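# Looping spot ambience sound cue for underground sewer pipe water, played in the Fort_Water sound class.
# A Mixer layers looping pipe, lake and cave water waves, with cave water boosted, under an endlessly looping Enveloper that slowly swells the volume.
# All quality levels share that chain.
# Overrides add capsule attenuation with low-pass filtering and listener focus, plus limited concurrency, and a send feeds the exterior ambient bus.
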SoundCue /Game/Sounds/Helios/Environment_Actors/SewerPipes/Spot_UndergroundPipe_Loop_03_Cue
root=QualityLevel_0
nodes [
 {
  id=Enveloper_2
  children [Mixer_1]
  properties {
   bLoopIndefinitely=true
   loopEnd=30
   pitchCurve {editorCurveData {keys [{time=0}]}}
   volumeCurve {editorCurveData {keys [<at 0 1 RCIM_Cubic>,<at 15 0.8 RCIM_Cubic>,<at 30 1 RCIM_Cubic>]}}
  }
 }
 {id=Mixer_1, children [WavePlayer_5,WavePlayer_0,WavePlayer_1], properties {inputVolume [1,1,2]}}
 {id=QualityLevel_0, children [Enveloper_2,Enveloper_2,Enveloper_2,Enveloper_2]}
 {id=WavePlayer_0, properties {bLooping=true, soundWaveAssetPtr=/Game/Sounds/Trans_AmbientLoops/Water/LakeB01_Loop}}
 {id=WavePlayer_1, properties {bLooping=true, soundWaveAssetPtr=/Game/Sounds/Fort_Starlight/S3/SoundWaves/Found_Cave_Water_C}}
 {
  id=WavePlayer_5
  properties {bLooping=true, soundWaveAssetPtr=&/Spot_UndergroundPipe_Loop_01}
 }
]
properties {
 attenuationSettings=&/Spot_UndergroundPipe_Loop_Capsule_Att
 bOverrideAttenuation=true
 bOverrideConcurrency=true
 maxDistance=1587
 soundClassObject=/Game/Sounds/FortSoundClassesAndModes/Fort_Water
 volumeMultiplier=0.8
 attenuationOverrides {
  attenuationShape=Capsule
  bAttenuateWithLPF=true
  bEnableFocusInterpolation=true
  bEnableListenerFocus=true
  bEnableLogFrequencyScaling=true
  falloffDistance=1200
  focusAttackInterpSpeed=0.1
  focusReleaseInterpSpeed=0.1
  focusVolumeAttenuation=1.2
  lPFFrequencyAtMax=2500
  lPFRadiusMax=2000
  lPFRadiusMin=750
  nonFocusVolumeAttenuation=0.5
  attenuationShapeExtents=<387 300>
 }
 busSends [{audioBus=null, minSendLevel=0.6, sendLevel=1, soundSourceBus=&AmbientAudio_Exterior_SourceBus}]
 concurrencyOverrides {maxCount=2}
 concurrencySet [&Weapon_Drop_StopOldest]
}
