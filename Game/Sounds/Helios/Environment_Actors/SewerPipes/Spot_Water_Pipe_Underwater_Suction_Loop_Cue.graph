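# Ambient Sound Cue for an underwater suction loop at sewer pipes, identical at every quality level.
# A mixer layers a quiet random drain loop, a slime loop under two nested 30-second envelopes, and two repeating, randomly delayed bubble one-shot branches.
# Modulators vary pitch per layer.
# Overridden attenuation adds listener focus and occlusion, overridden concurrency caps instances, and a bus send feeds the exterior ambient bus.
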SoundCue /Game/Sounds/Helios/Environment_Actors/SewerPipes/Spot_Water_Pipe_Underwater_Suction_Loop_Cue
root=QualityLevel_0
nodes [
 {id=Delay_0, children [Modulator_2], properties {delayMax=6, delayMin=2}}
 {id=Delay_1, children [Random_0], properties {delayMax=0.5}}
 {id=Delay_2, children [Modulator_3], properties {delayMax=6, delayMin=2}}
 {
  id=Enveloper_0
  children [Enveloper_2]
  properties {
   bLoopIndefinitely=true
   loopEnd=30
   pitchCurve {editorCurveData {keys [{interpMode=RCIM_Cubic, time=0}, <at 15 1.1 RCIM_Cubic>, {interpMode=RCIM_Cubic, time=30}]}}
   volumeCurve {editorCurveData {keys [<at 0 1 RCIM_Cubic>,<at 15 0.8 RCIM_Cubic>,<at 30 1 RCIM_Cubic>]}}
  }
 }
 {
  id=Enveloper_2
  children [Mixer_1]
  properties {
   bLoopIndefinitely=true
   loopEnd=30
   pitchCurve {editorCurveData {keys [{time=0}]}}
   volumeCurve {editorCurveData {keys [<at 0 1 RCIM_Cubic>,<at 15 0.8 RCIM_Cubic>,<at 30 1 RCIM_Cubic>]}}
  }
 }
 {id=Looping_0, children [Delay_0]}
 {id=Looping_1, children [Delay_2]}
 {id=Mixer_0, children [Modulator_1,Modulator_0,Looping_0,Looping_1], properties {inputVolume [0.2,1,2,1]}}
 {id=Mixer_1, children [WavePlayer_1], properties {inputVolume [2]}}
 {id=Modulator_0, children [Enveloper_0], properties {pitchMax=1.2, pitchMin=1.1, volumeMax=1.2, volumeMin=1.2}}
 {id=Modulator_1, children [Delay_1], properties {pitchMax=1.2, pitchMin=1.1, volumeMax=1.2, volumeMin=1.2}}
 {id=Modulator_2, children [Random_1], properties {pitchMax=1.1, pitchMin=0.9}}
 {id=Modulator_3, children [Random_2], properties {pitchMax=1.1, pitchMin=0.9}}
 {id=QualityLevel_0, children [Mixer_0,Mixer_0,Mixer_0,Mixer_0]}
 {id=Random_0, children [WavePlayer_0,WavePlayer_3], properties {weights [1,1]}}
 {id=Random_1, children [WavePlayer_8,WavePlayer_9], properties {weights [1,1]}}
 {id=Random_2, children [WavePlayer_13, WavePlayer_14], properties {weights [1,1]}}
 {
  id=WavePlayer_0
  properties {bLooping=true, soundWaveAssetPtr=/Game/Sounds/Asteria/Environment_Actors/WaterPipeOverflow/Spot_Water_PipeOverflow_DrainOpening_Medium_Loop}
 }
 {
  id=WavePlayer_1
  properties {bLooping=true, soundWaveAssetPtr=&/Spot_Water_Toxic_Sewer_Goop_Slime_Loop}
 }
 {
  id=WavePlayer_13
  properties {soundWaveAssetPtr=/Game/Sounds/Helios/Environment_Actors/NitroPipes/Spot_NitroPipe_OneShot_Metallic_Bubbles_02_Stereo}
 }
 {
  id=WavePlayer_14
  properties {soundWaveAssetPtr=/Game/Sounds/Helios/Environment_Actors/NitroPipes/Spot_NitroPipe_OneShot_Metallic_Bubbles_01_Stereo}
 }
 {
  id=WavePlayer_3
  properties {bLooping=true, soundWaveAssetPtr=/Game/Sounds/Asteria/Environment_Actors/WaterPipeOverflow/Spot_Water_PipeOverflow_DrainOpening_Large_Loop}
 }
 {id=WavePlayer_8, properties {soundWaveAssetPtr=/Game/Sounds/Asteria/Environment_Actors/FishTank/Spot_FishTank_Bubble_Air_Rush_03}}
 {id=WavePlayer_9, properties {soundWaveAssetPtr=/Game/Sounds/Asteria/Environment_Actors/FishTank/Spot_FishTank_Bubble_Air_Rush_04}}
]
properties {
 attenuationSettings=&/Spot_UndergroundPipe_Loop_Capsule_Att
 bEnableSubmixSends=false
 bOverrideAttenuation=true
 bOverrideConcurrency=true
 maxDistance=1750
 soundClassObject=&Fort_Ambient_Athena
 volumeMultiplier=0.15
 attenuationOverrides {
  bEnableListenerFocus=true
  falloffDistance=1250
  focusAzimuth=25
  focusVolumeAttenuation=1.2
  nonFocusAzimuth=45
  nonFocusVolumeAttenuation=0.75
  occlusionInterpolationTime=0.2
  occlusionLowPassFilterFrequency=1000
  occlusionTraceChannel=ECC_GameTraceChannel11
  occlusionVolumeAttenuation=0.5
  attenuationShapeExtents=<500 175>
 }
 busSends [{audioBus=null, minSendLevel=0.6, sendLevel=1, soundSourceBus=&AmbientAudio_Exterior_SourceBus}]
 concurrencyOverrides {maxCount=2}
 concurrencySet [&Weapon_Drop_StopOldest]
}
